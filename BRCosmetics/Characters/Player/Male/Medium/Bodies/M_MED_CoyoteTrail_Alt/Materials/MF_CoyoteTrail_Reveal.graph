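# Material function for the CoyoteTrail character body that animates a noisy spherical reveal with a glowing edge.
# A timer from TransitionTimeFromBP, reversible and optionally stepped, offsets Level_Progress within a start/stop range and samples a curve atlas to drive sphere radius and falloff.
# Pre-skinned-position sphere masks minus world-aligned noise yield pixel and vertex reveal masks, plus edge glow.
# It also outputs transition time and a vertex-color-gated WPO that collapses vertices toward the object position.
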MaterialFunction /BRCosmetics/Characters/Player/Male/Medium/Bodies/M_MED_CoyoteTrail_Alt/Materials/MF_CoyoteTrail_Reveal
graph {nodes [
  n1 = VectorParameter() {parameterName=EdgeGlow, defaultValue=<rgba 0.0182 0.5145 0.6979 1>}
  n2 = VectorParameter() {group=CoyoteTrail, parameterName=MaskPos, defaultValue=<rgba 0 30 150 1>}
  n3 = ScalarParameter() {group=CoyoteTrail, parameterName=Level_Progress}
  n4 = Time()
  n5 = ScalarParameter() {defaultValue=-100, group=CoyoteTrail, parameterName=TransitionTimeFromBP}
  n6 = Saturate(None=Divide(A=Subtract(A=@n4, B=@n5), B=ScalarParameter() {defaultValue=1.5, group=CoyoteTrail, parameterName=Animation_Duration}))
  n7 = LinearInterpolate(A=@n6, B=OneMinus(None=@n6), Alpha=ScalarParameter() {group=CoyoteTrail, parameterName=TransitionBackwards, sliderMax=1})
  n8 = LinearInterpolate(A=Subtract(A=@n3), B=@n3, Alpha=@n7)
  n9 = ScalarParameter() {group=CoyoteTrail, parameterName=Level_StartAt}
  n10 = ScalarParameter() {defaultValue=1, group=CoyoteTrail, parameterName=Level_StopAt}
  n11 = Divide(A=Subtract(A=Clamp(None=@n8, Min=@n9, Max=@n10), B=@n9), B=Subtract(A=@n10, B=@n9))
  n12 = Max(A=Ceil(None=ScalarParameter() {defaultValue=1, group=CoyoteTrail, parameterName=Steps}))
  n13 = Multiply(A=@n11, B=@n12)
  n14 = Subtract(A=@n12)
  n15 = LinearInterpolate(A=@n11, B=Divide(A=Floor(None=@n13), B=@n12), Alpha=Saturate(None=Ceil(None=@n14)))
  n16 = CurveAtlasRowParameter(CurveTime=@n15) {
   atlas=&AthenaCurveAtlas01
   curve=/Game/Athena/LinearColorCurves/Curve_SuperlevelMask
   defaultValue=205
   group=CoyoteTrail
   parameterName=MaskCurve
  }
  n17 = ChannelMaskParameter(None=@n16) {parameterName=CurveChannel}
  n18 = LinearInterpolate(
   A=ScalarParameter() {group=CoyoteTrail, parameterName=Radius_Start}
   B=ScalarParameter() {defaultValue=250, group=CoyoteTrail, parameterName=Radius_End}
   Alpha=@n17.R
  )
  n19 = ScalarParameter() {defaultValue=0.5, group=CoyoteTrail, parameterName=MaskHardness}
  n20 = VertexInterpolator(VS=PreSkinnedPosition())
  n21 = TextureObject() {samplerType=SAMPLERTYPE_LinearColor, texture=/Game/Characters/Player/Male/Medium/Bodies/M_MED_Polygon/Textures/M_MED_Polygon_SLMask}
  n22 = Constant3Vector() {constant=<rgba 175 175 175 0>}
  n23 = PreSkinnedNormal()
  n24 = ChannelMaskParameter(None=MaterialFunctionCall(
    TextureObject=NamedRerouteDeclaration(None=@n21) {name=textObj, nodeColor=<rgba 0.6641 1 0 1>}
    TextureSize=NamedRerouteDeclaration(None=@n22) {name=noiseTiling, nodeColor=<rgba 0.3828 1 0 1>}
    WorldPosition=@n20.PS
    World Space Normal=VertexInterpolator(VS=@n23).PS
   ) {materialFunction=/Game/Weapons/WeaponSkins/Wraps/MaterialFunctions/WorldAlignedTexture}.XYZ Texture) {parameterName=NoiseChannel}
  n25 = LinearInterpolate(
   A=ScalarParameter() {defaultValue=0.5, group=CoyoteTrail, parameterName=Falloff_Start}
   B=ScalarParameter() {defaultValue=0.1, group=CoyoteTrail, parameterName=Falloff_End}
   Alpha=Power(Base=@n17.R)
  )
  n26 = Subtract(
   A=SphereMask(
    A=@n20.PS
    B=NamedRerouteDeclaration(None=@n2.RGB) {name=MaskPosition, nodeColor=<rgba 1 0 0.1406 1>}
    Radius=NamedRerouteDeclaration(None=@n18) {name=MaskRadius, nodeColor=<rgba 0.2422 0 1 1>}
    Hardness=NamedRerouteDeclaration(None=@n19) {name=MaskHardness, nodeColor=<rgba 0 1 0.3672 1>}
   )
   B=Multiply(
    A=NamedRerouteDeclaration(None=@n24.R) {name=pattern, nodeColor=<rgba 0.2891 0 1 1>}
    B=NamedRerouteDeclaration(None=@n25) {name=fallOff, nodeColor=<rgba 0.3125 1 0 1>}
   )
  )
  n27 = Distance(A=@n26, B=ScalarParameter() {defaultValue=0.5, group=CoyoteTrail, parameterName=EdgeSample})
  n28 = ScalarParameter() {defaultValue=0.1, group=CoyoteTrail, parameterName=EdgeWidth}
  n29 = Clamp(None=OneMinus(None=Divide(A=@n27, B=@n28)))
  n30 = Power(
   Base=Multiply(A=Multiply(A=@n1.RGB, B=@n29), B=ScalarParameter() {defaultValue=5, group=CoyoteTrail, parameterName=EdgeStrength})
   Exp=ScalarParameter() {defaultValue=2, group=CoyoteTrail, parameterName=EdgePower}
  )
  n31 = NamedRerouteUsage()
  FunctionOutput(None=NamedRerouteDeclaration(None=Multiply(A=@n30, B=@n31)) {name=outEdgeMask, nodeColor=<rgba 0.4531 1 0 1>}) {outputName=EdgeMask}
  n32 = Round(None=@n26)
  FunctionOutput(None=NamedRerouteDeclaration(None=Saturate(None=@n32)) {name=outReveal, nodeColor=<rgba 1 0.5625 0 1>}) {outputName=Reveal}
  n33 = Subtract(A=WorldPosition().XYZ, B=ObjectPositionWS())
  n34 = Subtract(A=OneMinus(None=VertexColor().R)) {constB=0.01}
  FunctionOutput(None=Multiply(A=Multiply(A=@n33) {constB=-1}, B=Ceil(None=@n34))) {outputName=WPO}
  n35 = PreSkinnedPosition()
  n36 = SphereMask(A=@n35, B=NamedRerouteUsage(), Radius=NamedRerouteUsage(), Hardness=NamedRerouteUsage())
  n37 = NamedRerouteUsage()
  n38 = NamedRerouteUsage()
  n39 = Multiply(
   A=ChannelMaskParameter(None=MaterialFunctionCall(TextureObject=@n37, TextureSize=@n38, WorldPosition=@n35) {materialFunction=&WorldAlignedTexture}.XYZ Texture) {parameterName=NoiseChannel}.R
   B=NamedRerouteUsage()
  )
  n40 = Saturate(None=Round(None=Subtract(A=@n36, B=@n39)))
  FunctionOutput(None=@n40) {outputName=VertexShaderOutput}
  FunctionOutput(None=NamedRerouteUsage()) {outputName=transitionTime}
  NamedRerouteDeclaration(None=@n40) {name=vertexMask, nodeColor=<rgba 0 0.4609 1 1>}
  n41 = Subtract(A=Power(Base=@n7, Exp=ScalarParameter() {defaultValue=1.33, group=CoyoteTrail, parameterName=Animation_EaseIn})) {constB=0.5}
  n42 = OneMinus(None=Abs(None=Multiply(A=@n41) {constB=2}))
  NamedRerouteDeclaration(None=Power(Base=LinearInterpolate(A=@n42) {constAlpha=1}, Exp=ScalarParameter() {defaultValue=6, group=CoyoteTrail, parameterName=Animation_EaseOut})) {name=outEdgeTiming, nodeColor=<rgba 1 0.7266 0 1>}
  NamedRerouteDeclaration(None=@n15) {name=transitionTime, nodeColor=<rgba 0.5469 1 0 1>}
  TextureObjectParameter() {group=CoyoteTrail, parameterName=Noise, texture=/Game/Characters/Player/Male/Medium/Bodies/M_MED_Viper/Textures/Glitch/T_SoftNoise}
  VectorParameter() {group=CoyoteTrail, parameterName=Noise_Tiling, defaultValue=<rgba 24 24 24 1>}
 ]}
