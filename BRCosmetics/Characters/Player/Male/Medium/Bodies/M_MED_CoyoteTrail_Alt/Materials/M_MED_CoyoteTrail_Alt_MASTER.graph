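# Master material for the CoyoteTrail Alt skeletal character body, with an optional reactive reveal between a default look and the standard character shading.
# When enabled, attributes from the PreFX and SkinFX chain are overridden by lerps with DefaultLook textures, alongside MF_CoyoteTrail_Reveal, before PostFX and cinematic functions.
# Static switches add masked fresnel glow, skin-tint adjustment and world position offset that snaps at a threshold via Sign.
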Material /BRCosmetics/Characters/Player/Male/Medium/Bodies/M_MED_CoyoteTrail_Alt/Materials/M_MED_CoyoteTrail_Alt_MASTER
graph {
 properties {
  bAutomaticallySetUsageInEditor=false
  bUseMaterialAttributes=true
  bUsedWithMeshDeformer=true
  bUsedWithMorphTargets=true
  bUsedWithSkeletalMesh=true
  pixelDepthOffsetMode=PDOM_Legacy
  refractionMethod=RM_IndexOfRefraction
  shadingModel=MSM_FromMaterialExpression
  subsurfaceProfile=&SS_HeroSkin_02
  usedShadingModels=Default Lit | Subsurface | Clear Coat | Subsurface Profile | Cloth | Eye
 }
 nodes [
  n1 = MaterialFunctionCall() {materialFunction=&BaseFN_CharacterFunction_PreFX}
  n2 = MaterialFunctionCall(
   PreskinnedPosition=@n1.Pre skinned Local Position
   Gmap=@n1.Pre skinned Local Position
   MaterialAttributes=@n1.Pre skinned Local Position
   AO=@n1.Pre skinned Local Position
  ) {materialFunction=&CharacterSkinFX}
  n3 = GetMaterialAttributes(None=@n2.Result) {attributeGetTypes [
    #1
    #2
    #3
    #4
    #5
    #6
    #7
   ]}
  n4 = SetMaterialAttributes!(
   MaterialAttributes=@n3.MaterialAttributes
   Emissive Color=NamedRerouteUsage()
   Base Color=NamedRerouteUsage()
   Specular=NamedRerouteUsage()
   Roughness=NamedRerouteUsage()
   Curvature or Metallic=NamedRerouteUsage()
   Normal=NamedRerouteUsage()
   World Position Offset=StaticSwitchParameter(True=NamedRerouteUsage(), False=NamedRerouteDeclaration(None=@n3.World Position Offset) {name=inWPO, nodeColor=<rgba 0.6172 1 0 1>}) {group=CoyoteTrail, parameterName=useWPO}
  ) {attributeSetTypes [
    #1
    #2
    #3
    #4
    #5
    #6
    #7
   ]}
  n5 = MaterialFunctionCall(
   AO=@n1.AO
   Distance Blend=@n1.AO
   Fresnel=@n1.AO
   Gmap=@n1.AO
   Pre Skinned Local Position=@n1.AO
   Material=StaticSwitchParameter(True=@n4, False=@n2.Result) {group=CoyoteTrail, parameterName=useCoyoteReactive}
  ) {materialFunction=&BaseFN_CharacterFunction_PostFX}
  MaterialFunctionCall(In=@n5.Result) {materialFunction=&CINE_MF_CharPostFX}
  n6 = MaterialFunctionCall() {materialFunction=&/MF_CoyoteTrail_Reveal}
  NamedRerouteDeclaration(None=@n6.Reveal) {name=outReveal, nodeColor=<rgba 1 0.5625 0 1>}
  NamedRerouteDeclaration(None=@n6.EdgeMask) {name=outEdgeMask, nodeColor=<rgba 0.4531 1 0 1>}
  NamedRerouteDeclaration(None=@n3.Specular) {name=inSpec, nodeColor=<rgba 1 0.7734 0 1>}
  n7 = TextureSampleParameter2D() {
   group=CoyoteTrail
   parameterName=DefaultLook_Spec
   samplerType=SAMPLERTYPE_Masks
   texture=/BRCosmetics/Characters/Player/Male/Medium/Bodies/M_MED_CoyoteTrail/Textures/T_Character_CoyoteTrail_BODY_S
  }
  NamedRerouteDeclaration(None=LinearInterpolate(A=@n7.R, B=NamedRerouteUsage(), Alpha=NamedRerouteUsage())) {name=outSpec, nodeColor=<rgba 0 1 0.5078 1>}
  NamedRerouteDeclaration(None=LinearInterpolate(A=@n7.B, B=NamedRerouteUsage(), Alpha=NamedRerouteUsage())) {name=outRoughness, nodeColor=<rgba 0 0.1797 1 1>}
  NamedRerouteDeclaration(None=@n3.Roughness) {name=inRoughness, nodeColor=<rgba 0 0.2031 1 1>}
  NamedRerouteDeclaration(None=@n3.Curvature or Metallic) {name=inMetalic, nodeColor=<rgba 0.5 0 1 1>}
  NamedRerouteDeclaration(None=LinearInterpolate(A=@n7.G, B=NamedRerouteUsage(), Alpha=NamedRerouteUsage())) {name=outMetalic, nodeColor=<rgba 0.7578 1 0 1>}
  n8 = TextureSampleParameter2D() {
   group=CoyoteTrail
   parameterName=DefaultLook_Normal
   samplerType=SAMPLERTYPE_Normal
   texture=/BRCosmetics/Characters/Player/Male/Medium/Bodies/M_MED_CoyoteTrail/Textures/T_Character_CoyoteTrail_BODY_N
  }
  NamedRerouteDeclaration(None=LinearInterpolate(A=@n8.RGB, B=NamedRerouteUsage(), Alpha=NamedRerouteUsage())) {name=outNormal, nodeColor=<rgba 1 0.375 0 1>}
  NamedRerouteDeclaration(None=@n3.Normal) {name=inNormal, nodeColor=<rgba 1 0 0.4453 1>}
  NamedRerouteDeclaration(None=@n6.VertexShaderOutput) {name=VertexShader, nodeColor=<rgba 0 0.1094 1 1>}
  NamedRerouteDeclaration(None=@n6.transitionTime) {name=transitionTime, nodeColor=<rgba 0 1 0.3672 1>}
  NamedRerouteDeclaration(None=@n3.Emissive Color) {name=inEC, nodeColor=<rgba 1 0.4219 0 1>}
  NamedRerouteDeclaration(None=@n3.Base Color) {name=inBC, nodeColor=<rgba 0 1 0.5313 1>}
  NamedRerouteDeclaration(None=LinearInterpolate(A=NamedRerouteUsage(), B=NamedRerouteUsage(), Alpha=NamedRerouteUsage()) {constAlpha=1}) {name=outBC, nodeColor=<rgba 0.9219 0 1 1>}
  n9 = TextureSampleParameter2D() {
   group=CoyoteTrail
   parameterName=DefaultLook_Emissive
   texture=/BRCosmetics/Characters/Player/Male/Medium/Bodies/M_MED_CoyoteTrail/Textures/T_Character_CoyoteTrail_BODYACC_E
  }
  n10 = NamedRerouteUsage()
  n11 = VectorParameter() {group=CoyoteTrail, parameterName=Fresnel Color, defaultValue=<rgba 0.0648 0.5711 0.1441 1>}
  n12 = TextureSampleParameter2D() {
   group=CoyoteTrail
   parameterName=Emissive Mask
   samplerType=SAMPLERTYPE_Masks
   texture=/BRCosmetics/Characters/Player/Male/Medium/Bodies/M_MED_CoyoteTrail_Alt/Textures/T_Character_CoyoteTrailAlt_BodyAcc_FX
  }
  n13 = Fresnel(ExponentIn=ScalarParameter() {defaultValue=2, group=CoyoteTrail, parameterName=Fresnel_Size}) {baseReflectFraction=0}
  n14 = Multiply(
   A=Multiply(A=StaticComponentMaskParameter(None=@n12.RGB) {defaultG=true, group=CoyoteTrail, parameterName=Emisive Channel}, B=Saturate(None=@n13))
   B=ScalarParameter() {defaultValue=5, group=CoyoteTrail, parameterName=Fresnel_Power}
  ) {constB=25}
  n15 = StaticSwitchParameter(True=Add(A=@n10, B=Multiply(A=@n11.RGB, B=@n14)), False=@n10) {defaultValue=true, group=CoyoteTrail, parameterName=useFresnel}
  n16 = NamedRerouteUsage()
  NamedRerouteDeclaration(None=LinearInterpolate(A=@n9.RGB, B=Add(A=@n15, B=@n16), Alpha=NamedRerouteUsage()) {constAlpha=0, constB=0}) {name=outEC, nodeColor=<rgba 0.1953 0 1 1>}
  NamedRerouteDeclaration(None=@n6.WPO) {name=inPixelWPO, nodeColor=<rgba 0 0.5547 1 1>}
  n17 = NamedRerouteUsage()
  n18 = Sign(None=Subtract(A=ComponentMask(None=@n17) {r=true}, B=ScalarParameter() {defaultValue=0.75, group=CoyoteTrail, parameterName=vertexSnapInValue}) {constB=0.8})
  NamedRerouteDeclaration(None=LinearInterpolate(A=NamedRerouteUsage(), B=NamedRerouteUsage(), Alpha=Saturate(None=@n18))) {name=outWPO, nodeColor=<rgba 1 0 0.8906 1>}
  n19 = ScalarParameter() {defaultValue=0.75, parameterName=Global Skin Boost EX}
  n20 = VectorParameter() {group=SkinColor, parameterName=Skin Boost Color And Exponent, defaultValue=<rgba 0.87 0.6688 0.6948 2>}
  n21 = Fresnel(ExponentIn=@n20.A) {exponent=2}
  n22 = TextureSampleParameter2D() {
   group=CoyoteTrail
   parameterName=DefaultLook_Diffuse
   texture=/BRCosmetics/Characters/Player/Male/Medium/Bodies/M_MED_CoyoteTrail/Textures/T_Character_CoyoteTrail_BODY_D
  }
  n23 = Multiply(A=@n22.RGB, B=@n20.RGB)
  n24 = Add(A=Multiply(A=OneMinus(None=@n21), B=@n23), B=@n23)
  n25 = Saturate(None=LinearInterpolate(A=@n22.RGB, B=Multiply(A=@n19, B=@n24), Alpha=ScalarParameter() {defaultValue=1, group=SkinColor, parameterName=HumanSkin}))
  n26 = TextureSampleParameter2D() {
   group=CoyoteTrail
   parameterName=SkinMask
   samplerType=SAMPLERTYPE_Masks
   texture=/BRCosmetics/Characters/Player/Male/Medium/Bodies/M_MED_CoyoteTrail/Textures/T_Character_CoyoteTrail_BODY_M
  }
  n27 = StaticComponentMaskParameter(None=@n26.B) {defaultB=true, group=CoyoteTrail, parameterName=SkinMaskChannel}
  NamedRerouteDeclaration(None=StaticSwitchParameter(True=LinearInterpolate(A=@n22.RGB, B=@n25, Alpha=@n27) {constAlpha=1, constB=0}, False=@n22.RGB) {group=CoyoteTrail, parameterName=useSkinAdjustment}) {name=SkinAdjust, nodeColor=<rgba 0 1 0.2031 1>}
  NamedRerouteUsage()
 ]
}
