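# Master material for the CoyoteTrail Alt player body cosmetic, built on material attributes with per-expression shading models.
# Shared character functions chain PreFX, skin FX, PostFX and cinematic PostFX.
# Between skin FX and PostFX, the useCoyoteFresnel static switch can override emissive, routed through named reroutes.
# That branch adds a tinted Fresnel glow, masked by a selectable FX texture channel, with adjustable size and power, to the existing emissive.
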Material /BRCosmetics/Characters/Player/Male/Medium/Bodies/M_MED_CoyoteTrail_Alt/Materials/M_MED_CoyoteTrail_Alt_NonReactive_MASTER
graph {
 properties {
  bAutomaticallySetUsageInEditor=false
  bUseMaterialAttributes=true
  bUsedWithMeshDeformer=true
  bUsedWithMorphTargets=true
  bUsedWithSkeletalMesh=true
  pixelDepthOffsetMode=PDOM_Legacy
  refractionMethod=RM_IndexOfRefraction
  shadingModel=MSM_FromMaterialExpression
  subsurfaceProfile=&SS_HeroSkin_02
  usedShadingModels=Default Lit | Subsurface | Clear Coat | Subsurface Profile | Cloth | Eye
 }
 nodes [
  n1 = MaterialFunctionCall() {materialFunction=&BaseFN_CharacterFunction_PreFX}
  n2 = MaterialFunctionCall(
   PreskinnedPosition=@n1.Pre skinned Local Position
   Gmap=@n1.Pre skinned Local Position
   MaterialAttributes=@n1.Pre skinned Local Position
   AO=@n1.Pre skinned Local Position
  ) {materialFunction=&CharacterSkinFX}
  n3 = GetMaterialAttributes(None=@n2.Result) {attributeGetTypes [#1]}
  n4 = NamedRerouteUsage()
  n5 = MaterialFunctionCall(
   AO=@n1.AO
   Distance Blend=@n1.AO
   Fresnel=@n1.AO
   Gmap=@n1.AO
   Pre Skinned Local Position=@n1.AO
   Material=StaticSwitchParameter(
    True=SetMaterialAttributes(MaterialAttributes=@n3.MaterialAttributes, Emissive Color=@n4) {attributeSetTypes [#1]}
    False=@n2.Result
   ) {group=CoyoteTrail, parameterName=useCoyoteFresnel}
  ) {materialFunction=&BaseFN_CharacterFunction_PostFX}
  MaterialFunctionCall(In=@n5.Result) {materialFunction=&CINE_MF_CharPostFX}
  NamedRerouteDeclaration(None=@n3.Emissive Color) {name=inEmissive, nodeColor=<rgba 1 0 0.5391 1>}
  n6 = VectorParameter() {group=CoyoteTrail, parameterName=Fresnel Color, defaultValue=<rgba 0.0648 0.5711 0.1441 1>}
  n7 = TextureSampleParameter2D() {
   group=CoyoteTrail
   parameterName=Emissive Mask
   samplerType=SAMPLERTYPE_Masks
   texture=/BRCosmetics/Characters/Player/Male/Medium/Bodies/M_MED_CoyoteTrail_Alt/Textures/T_Character_CoyoteTrailAlt_BodyAcc_FX
  }
  n8 = Fresnel(ExponentIn=ScalarParameter() {defaultValue=2, group=CoyoteTrail, parameterName=Fresnel_Size}) {baseReflectFraction=0}
  n9 = Multiply(
   A=Multiply(A=StaticComponentMaskParameter(None=@n7.RGB) {defaultG=true, group=CoyoteTrail, parameterName=Emisive Channel}, B=Saturate(None=@n8))
   B=ScalarParameter() {defaultValue=5, group=CoyoteTrail, parameterName=Fresnel_Power}
  ) {constB=25}
  NamedRerouteDeclaration(None=Add(A=NamedRerouteUsage(), B=Multiply(A=@n6.RGB, B=@n9))) {name=outEmissive, nodeColor=<rgba 1 0.5391 0 1>}
 ]
}
