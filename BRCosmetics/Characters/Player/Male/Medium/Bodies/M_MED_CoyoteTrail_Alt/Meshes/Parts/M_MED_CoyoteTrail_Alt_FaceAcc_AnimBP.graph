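# Post-copy Animation Blueprint for the face accessory mesh of the male CoyoteTrail_Alt character, adding head offsets and earring physics.
# It copies the attached parent's pose, then layers additive head bone rotations for down, left and right.
# Each update derives their alphas from the head's reference-pose delta against neck_02; the up alpha is computed but unused.
# Two half-weight, spring-driven Anim Dynamics nodes then swing the left and right earrings.
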AnimBlueprint /BRCosmetics/Characters/Player/Male/Medium/Bodies/M_MED_CoyoteTrail_Alt/Meshes/Parts/M_MED_CoyoteTrail_Alt_FaceAcc_AnimBP
target_skeleton=&/M_MED_CoyoteTrail_Alt_FaceAcc_Skeleton
graphs [
 AnimGraph {nodes [
   n1 = "Transform (Modify) Bone - Bone: head"(
    ComponentPose=Local To Component(LocalPose=Copy Pose From Mesh() {bUseAttachedParent=true})
    Alpha=Get neck_rot_dn() -> (neck_rot_dn:double)
    Translation="0,0,0"
    Rotation="0,-30,0"
    Scale="1,1,1"
   ) {
    rotationMode=BMM_Additive
    rotationSpace=BCS_BoneSpace
    scaleMode=BMM_Additive
    scaleSpace=BCS_BoneSpace
    translationMode=BMM_Additive
    translationSpace=BCS_BoneSpace
    alphaBoolBlend {blendOption=Linear}
    boneToModify {boneName=head}
    rotation {yaw=-30}
   }
   n2 = "Transform (Modify) Bone - Bone: head"(ComponentPose=@n1, Alpha=Get neck_rot_l() -> (neck_rot_l:double), Translation="0,0,0", Rotation="0,0,-15", Scale="1,1,1") {
    rotationMode=BMM_Additive
    rotationSpace=BCS_BoneSpace
    scaleMode=BMM_Additive
    scaleSpace=BCS_BoneSpace
    translationMode=BMM_Additive
    translationSpace=BCS_BoneSpace
    alphaBoolBlend {blendOption=Linear}
    boneToModify {boneName=head}
    rotation {roll=-15}
   }
   n3 = "Transform (Modify) Bone - Bone: head"(ComponentPose=@n2, Alpha=Get neck_rot_r() -> (neck_rot_r:double), Translation="0,0,0", Rotation="0,0,15", Scale="1,1,1") {
    rotationMode=BMM_Additive
    rotationSpace=BCS_BoneSpace
    scaleMode=BMM_Additive
    scaleSpace=BCS_BoneSpace
    translationMode=BMM_Additive
    translationSpace=BCS_BoneSpace
    alphaBoolBlend {blendOption=Linear}
    boneToModify {boneName=head}
    rotation {roll=15}
   }
   n4 = Anim Dynamics - Bone: dyn_Earrings_l(ComponentPose=@n3, Alpha=0.5, ExternalForce="0,0,0") {
    angularDampingOverride=0.85
    angularSpringConstant=800
    bAngularSpring=true
    bGravityOverrideInSimSpace=true
    bOverrideAngularDamping=true
    bUseGravityOverride=true
    gravityScale=0.3
    simulationSpace=RootRelative
    alphaBoolBlend {blendOption=Linear}
    boundBone {boneName=dyn_Earrings_l}
    componentAppliedLinearAccClamp=<1200 1200 1200>
    componentLinearAccScale=<0.2 0.2 0.25>
    componentLinearVelScale=<0.25 0.25 0.25>
    gravityOverride {z=-980}
    physicsBodyDefinitions [{
      sphereCollisionRadius=0
      boundBone {boneName=dyn_Earrings_l}
      boxExtents=<20 20 20>
      constraintSetup {twistAxis=AxisY, angularLimitsMax=<30 0.1 1>, angularLimitsMin=<-16 -0.1 -1>, angularTarget=<0 0 0>}
      localJointOffset=<0.01 -0.5 -5>
     }]
    relativeSpaceBone {boneName=spine_05}
   }
   n5 = Anim Dynamics - Bone: dyn_Earrings_r(ComponentPose=@n4, Alpha=0.5, ExternalForce="0,0,0") {
    angularDampingOverride=0.85
    angularSpringConstant=800
    bAngularSpring=true
    bGravityOverrideInSimSpace=true
    bOverrideAngularDamping=true
    bUseGravityOverride=true
    gravityScale=0.3
    simulationSpace=RootRelative
    alphaBoolBlend {blendOption=Linear}
    boundBone {boneName=dyn_Earrings_r}
    componentAppliedLinearAccClamp=<1200 1200 1200>
    componentLinearAccScale=<0.2 0.2 0.25>
    componentLinearVelScale=<0.25 0.25 0.25>
    gravityOverride {z=-980}
    physicsBodyDefinitions [{
      sphereCollisionRadius=0
      boundBone {boneName=dyn_Earrings_r}
      boxExtents=<18 18 18>
      constraintSetup {twistAxis=AxisY, angularLimitsMax=<30 0.1 1>, angularLimitsMin=<-16 -0.1 -1>, angularTarget=<0 0 0>}
      localJointOffset=<0.01 -0.1 -5>
     }]
    relativeSpaceBone {boneName=spine_05}
   }
   Output Pose(Result=Component To Local(ComponentPose=@n5)) {layerGroup=None}
  ]}
 EventGraph {nodes [
   Event BlueprintUpdateAnimation()
   > Set neck_rot_up(neck_rot_up:double=GetAlphaFromReferenceTransformDelta(SkeletalMeshComponent=Get Skel_Mesh(), BoneName=head, BaseName=neck_02, InputRange=(X=0,Y=-40), TwistAxis="0,0,1")) -> (Output_Get:double)
   > Set neck_rot_dn(neck_rot_dn:double=GetAlphaFromReferenceTransformDelta(SkeletalMeshComponent=Get Skel_Mesh(), BoneName=head, BaseName=neck_02, InputRange=(X=-15,Y=30), TwistAxis="0,0,1")) -> (Output_Get:double)
   > Set neck_rot_l(neck_rot_l:double=GetAlphaFromReferenceTransformDelta(SkeletalMeshComponent=Get Skel_Mesh(), BoneName=head, BaseName=neck_02, DeltaComponentType=EulerX, InputRange=(X=0,Y=40), TwistAxis="0,0,1")) -> (Output_Get:double)
   > Set neck_rot_r(neck_rot_r:double=GetAlphaFromReferenceTransformDelta(SkeletalMeshComponent=Get Skel_Mesh(), BoneName=head, BaseName=neck_02, DeltaComponentType=EulerX, InputRange=(X=0,Y=-40), TwistAxis="0,0,1")) -> (Output_Get:double)
   Event BlueprintInitializeAnimation()
   > Set Skel_Mesh(Skel_Mesh=GetOwnerSkeletalMeshComponent())
  ]}
]
variables [
 neck_rot_dn {type=number}
 neck_rot_l {type=number}
 neck_rot_r {type=number}
 neck_rot_up {type=number}
 Skel_Mesh {type=SkeletalMeshComponent}
]
